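; Creative minigame settings device that turns island game options into configuration of the current minigame, editable by players allowed to create in the volume.
; On authority, option load or change events run Update functions that push values to the minigame, its components, mutators and debug/physics scalable floats.
; A DoOnce-guarded upgrade converts legacy game-mode presets into property overrides without re-trigger loops.
; Entering Setup clears temporary Game End device callouts and reapplies post-game settings.
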(event Custom|OnMinigameStateChanged
  (params (Minigame :type Fort%20Minigame%20Object%20Reference) (MinigameState :type EFortMinigameState%20Enum :default "PreGame"))
  (bind _returnvalue (Minigame|GetMinigameSettingsMachine Minigame))
  (if (Utilities|Enum|Equal(Enum) MinigameState "Setup")
    (CallFunction|OverrideCustomEndCallouts "" "" (Variables|Default|GetGameEndCallout)) ; We're setting up a new game / round, reset the Device Vicotry Callouts \n(there are temporary overrides of minigame settings by the Game End device)
    ;[1 1 1] Reset the PostGameSettings to initial island values.
    (Utilities|IsValid Minigame
      (:Is%20Valid
        (Utilities|IsValid _returnvalue
          (:Is%20Valid
            (Minigame|OverridePostGameSettings _returnvalue (Variables|/Script/FortniteGame.FortMinigameVictoryScreenComponent|GetPostGameSettings (Variables|/Script/FortniteGame.FortMinigame|GetVictoryScreenComponent Minigame)))))))))

(event EventBeginPlay
  (bind _minigame (CallFunction|GetCurrentMinigame))
  (|CallParentFunction)
  (Networking|SwitchHasAuthority
    (:Authority
      (OptionData|AssignOnAnyPropertyChanged (Variables|Default|GetToyOptionsComponent) (EventDispatchers|CreateEvent :!event Custom|OnGameSettingsUpdated))
      (OptionData|AssignOnPropertiesLoaded (Variables|Default|GetToyOptionsComponent) (EventDispatchers|CreateEvent :!event Custom|OnGameSettingsUpdated))
      (CallFunction|FillTrackedStats)
      (Utilities|IsValid _minigame
        (:Is%20Valid
          (Minigame|BindEventtoStateChangedDelegate _minigame (delegate Custom|OnMinigameStateChanged)))))))

(event Custom|OnGameSettingsUpdated
  (Utilities|FlowControl|Sequence
    (:then_0
      (CallFunction|UpdateGameMode)
      (CallFunction|UpdateClassSettings)
      (CallFunction|UpdateTimeLimit)
      (CallFunction|UpdateTotalRounds)
      (CallFunction|UpdateSpawnLocationSetting)
      (CallFunction|UpdateTeamSetting)
      (CallFunction|UpdatePostGameSpawnLocationSetting)
      (CallFunction|UpdateSpawnPadSelection)
      (CallFunction|UpdatePlayerLives)
      (CallFunction|UpdateAutoEndTeamThreshold)
      (CallFunction|UpdateTeamtoMovetowhenOutOfSpawns)
      (CallFunction|UpdateEndConditions)
      (CallFunction|UpdateGameStartCountdown)
      (CallFunction|UpdateAutoStart)
      (CallFunction|UpdateJoininProgress)
      (CallFunction|UpdateMinigameMutators)
      (CallFunction|UpdatePhysicsSettings)
      (Utilities|FlowControl|DoOnce false
        (:Completed
          (Utilities|FlowControl|Sequence
            (:then_0
              _)
            (:then_1
              _))
          (if (!= (Variables|Default|GetGameModeIndex) -1)
            (CallFunction|GameModePresetUpgradePath)
            (Variables|Default|SetGameModeIndex -1)))) ; Protect against infinite loop. Code past this point makes changes\nto property overrides which will cause duplicate triggers of this event.
    )
    (:then_1
      (CallFunction|UpdateGameSettings)
      (CallFunction|UpdateDebugSettings))))

(event Minigame|EventStartMinigame_BP
  (bind _asfort_player_controller_athena (Utilities|Casting|CastToFortPlayerControllerAthena (Utilities|Array|Get(acopy) (Fortnite|GetLocalFortPlayerControllers true))))
  (Creative|ServerStartMinigame _asfort_player_controller_athena))

(event Minigame|EventAbandonMinigame_BP
  (bind _asfort_player_controller_athena (Utilities|Casting|CastToFortPlayerControllerAthena (Utilities|Array|Get(acopy) (Fortnite|GetLocalFortPlayerControllers true))))
  (Creative|ServerEndMinigame _asfort_player_controller_athena false))

(fn AdditionalUIExtensionValidityChecks
  (params (TagContainer :type Gameplay%20Tag%20Container%20Structure))
  (Utilities|FlowControl|Sequence
    (:then_0
      (if (GameplayTags|HasTag (Variables|GetTagContainer) "(TagName=\"Minigame.UIExtension.RoundStatus\")" true)
        (return (== (Variables|Default|GetTeamSetting) 2))))
    (:then_1
      (return true))))

(fn BlueprintCanInteract
  (params (InteractingPawn :type Fort%20Pawn%20Object%20Reference) (InteractionBeingAttempted :type EInteractionBeingAttempted%20Enum :default "FirstInteraction") (InteractionType :type TInteractionType%20Enum :default "IT_NoInteraction"))
  (bind _returnvalue (Pawn|GetController InteractingPawn))
  (Utilities|FlowControl|Sequence
    (:then_0
      (Utilities|IsValid InteractingPawn
        (:Is%20Valid
          (Utilities|IsValid _returnvalue
            (:Is%20Valid
              (bind _asfort_player_controller_athena (Utilities|Casting|CastToFortPlayerControllerAthena _returnvalue))
              (LocalVariables|SetCanInteract (.ReturnValue (Creative|CanPlayerCreateinCurrentVolume _asfort_player_controller_athena :!class /Script/FortniteGame.FortPlayerControllerAthena))))))))
    (:then_1
      (return (LocalVariables|GetCanInteract)))))

(fn BlueprintGetFailedInteractionString
  (params (InteractingPawn :type Fort%20Pawn%20Object%20Reference) (InteractionBeingAttempted :type EInteractionBeingAttempted%20Enum :default "FirstInteraction"))
  (return "NSLOCTEXT(\"[9235D69282A3B83229F433882100160C]\", \"E568E62F4791A08E0D6DDEAD87C8C1FF\", \"Cannot Create Here\")"))

(fn BlueprintGetInteractionString
  (params (InteractingPawn :type Fort%20Pawn%20Object%20Reference) (InteractionBeingAttempted :type EInteractionBeingAttempted%20Enum :default "FirstInteraction"))
  (return "NSLOCTEXT(\"[9235D69282A3B83229F433882100160C]\", \"FA888245445B85865C98059A3953E2A7\", \"Modify Game Settings\")"))

(fn FillTrackedStats
  (bind _trackedcolumnstat4 (Variables|Default|GetTrackedColumnStat4))
  (bind _trackedcolumnstat3 (Variables|Default|GetTrackedColumnStat3))
  (bind _trackedcolumnstat2 (Variables|Default|GetTrackedColumnStat2))
  (bind _trackedcolumnstat1 (Variables|Default|GetTrackedColumnStat1))
  (bind _trackedcolumnstat0 (Variables|Default|GetTrackedColumnStat0))
  (bind _array (Utilities|Array|MakeArray _trackedcolumnstat0 _trackedcolumnstat1 _trackedcolumnstat2 _trackedcolumnstat3 _trackedcolumnstat4))
  (bind _output_get (Variables|Default|SetTrackedScoreStats (Utilities|Array|MakeArray (Variables|Default|GetTrackedScoreStat0) (Variables|Default|GetTrackedScoreStat1) (Variables|Default|GetTrackedScoreStat2) (Variables|Default|GetTrackedScoreStat3) (Variables|Default|GetTrackedScoreStat4))))
  ;[1 1 1] Only override the column order if there's at least one scoreboard column option set
  (Variables|Default|SetTrackedScoreColumnOrder _output_get)
  (bind _array_element (Utilities|Array|ForEachLoopwithBreak _array)
    (:LoopBody
      (if (>= _array_element 0)
        (Variables|Default|SetTrackedScoreColumnOrder _array)))))

(fn GameModePresetUpgradePath
  (bind _gamemodeindex (Variables|Default|GetGameModeIndex))
  (bind _output (Utilities|GetScoreingPreset _gamemodeindex))
  (bind (_sortscoreboard_6_afc328db4172cce4e5742a9504d0c087 _scoreboardstats_7_279a843f4d5a831e0cafe5ab87e85fb6 _uiextensions_17_755f8b1c4db6c595f20783a14b316aec) (Utilities|Struct|BreakMinigameScoringPreset _output))
  (bind (_option_key _option_value_name) (Utilities|GetOptionKeyandValueName (Variables|Default|GetTrackedScoreStat1_Definition) (Utilities|String|ToString(Integer) (Variables|Default|GetTrackedScoreStat1))))
  (bind (_option_key_2 _option_value_name_2) (Utilities|GetOptionKeyandValueName (Variables|Default|GetTrackedScoreStat2_Definition) (Utilities|String|ToString(Integer) (Variables|Default|GetTrackedScoreStat2))))
  (bind (_option_key_3 _option_value_name_3) (Utilities|GetOptionKeyandValueName (Variables|Default|GetTrackedScoreStat3_Definition) (Utilities|String|ToString(Integer) (Variables|Default|GetTrackedScoreStat3))))
  (bind (_option_key_4 _option_value_name_4) (Utilities|GetOptionKeyandValueName (Variables|Default|GetTrackedScoreStat4_Definition) (Utilities|String|ToString(Integer) (Variables|Default|GetTrackedScoreStat4))))
  (bind (_option_key_5 _option_value_name_5) (Utilities|GetOptionKeyandValueName (Variables|Default|GetUIExtensionSetting_Definition) (Utilities|String|ToString(Integer) (Variables|Default|GetUIExtensionSetting))))
  (bind (_option_key_6 _option_value_name_6) (Utilities|GetOptionKeyandValueName (Variables|Default|GetTrackedScoreStat0_Definition) (Utilities|String|ToString(Integer) (Variables|Default|GetTrackedScoreStat0))))
  (bind _toyoptionscomponent (Variables|Default|GetToyOptionsComponent))
  (if (Utilities|IsValidGameModePreset _gamemodeindex)
    (Utilities|FlowControl|Sequence
      (:then_0
        (Utilities|FlowControl|Sequence
          (:then_0
            (bind _output_2 (Utilities|IsValidIndex _scoreboardstats_7_279a843f4d5a831e0cafe5ab87e85fb6)
              (:true
                (Variables|Default|SetTrackedScoreStat0 (Utilities|GetTrackedStatIndex _output_2)))
              (:false
                (Variables|Default|SetTrackedScoreStat0 -1))))
          (:then_1
            (bind _output_3 (Utilities|IsValidIndex _scoreboardstats_7_279a843f4d5a831e0cafe5ab87e85fb6 1)
              (:true
                (Variables|Default|SetTrackedScoreStat1 (Utilities|GetTrackedStatIndex _output_3)))
              (:false
                (Variables|Default|SetTrackedScoreStat1 -1))))
          (:then_2
            (bind _output_4 (Utilities|IsValidIndex _scoreboardstats_7_279a843f4d5a831e0cafe5ab87e85fb6 2)
              (:true
                (Variables|Default|SetTrackedScoreStat2 (Utilities|GetTrackedStatIndex _output_4)))
              (:false
                (Variables|Default|SetTrackedScoreStat2 -1))))
          (:then_3
            (bind _output_5 (Utilities|IsValidIndex _scoreboardstats_7_279a843f4d5a831e0cafe5ab87e85fb6 3)
              (:true
                (Variables|Default|SetTrackedScoreStat3 (Utilities|GetTrackedStatIndex _output_5)))
              (:false
                (Variables|Default|SetTrackedScoreStat3 -1))))
          (:then_4
            (bind _output_6 (Utilities|IsValidIndex _scoreboardstats_7_279a843f4d5a831e0cafe5ab87e85fb6 4)
              (:true
                (Variables|Default|SetTrackedScoreStat4 (Utilities|GetTrackedStatIndex _output_6)))
              (:false
                (Variables|Default|SetTrackedScoreStat4 -1))))
          (:then_5
            (CallFunction|FillTrackedStats))))
      (:then_1
        (Variables|Default|SetUIExtensionSetting (Utilities|GetUIExtensionIndex _uiextensions_17_755f8b1c4db6c595f20783a14b316aec)))
      (:then_2
        (OptionData|SetPropertyOverrides _toyoptionscomponent (Utilities|Map|MakeMap _option_key_6 _option_value_name_6 _option_key _option_value_name _option_key_2 _option_value_name_2 _option_key_3 _option_value_name_3 _option_key_4 _option_value_name_4 _option_key_5 _option_value_name_5))
        (OptionData|RemovePropertyOverrides _toyoptionscomponent (Utilities|Array|MakeArray "GameModeIndex"))))
    (else
      (Development|PrintString "MinigameSettingsMachine::GameModePresetUpgradePath - Failed"))))

(fn GetScoreboardOrderObjects
  (params (TrackedStatIndices :type Array%20of%20Integers) (TrackedStats :type Array%20of%20Fort%20Minigame%20Stat%20Filter%20Object%20References))
  (bind _newlocalvar (LocalVariables|GetNewLocalVar))
  (for _array_element TrackedStatIndices
    (if (Utilities|Array|IsValidIndex TrackedStats _array_element)
      (Utilities|Array|AddUnique _newlocalvar (Utilities|Array|Get(acopy) TrackedStats _array_element))))
  (return _newlocalvar))

(fn GetCurrentMinigame
  (return (Minigame|GetMinigame)))

(fn GetDynamicUITags
  (Utilities|FlowControl|Sequence
    (:then_0
      (if (Variables|Default|GetShowTopCenterScoreHUD)
        (GameplayTags|AddGameplayTag (LocalVariables|GetOutDynamicUIExtensions) (Variables|Default|GetTopCenterExtensionTag) :!class /Script/GameplayTags.BlueprintGameplayTagLibrary)))
    (:then_1
      (if (Variables|Default|GetbShouldDisplayScoreboard)
        (GameplayTags|AddGameplayTag (LocalVariables|GetOutDynamicUIExtensions) (Variables|Default|GetScoreboardContainerExtensionTag) :!class /Script/GameplayTags.BlueprintGameplayTagLibrary)))
    (:then_2
      (if (Variables|Default|GetbShouldDisplayCareerScoreboard)
        (GameplayTags|AddGameplayTag (LocalVariables|GetOutDynamicUIExtensions) (Variables|Default|GetCareerScoreboardContainerExtensionTag) :!class /Script/GameplayTags.BlueprintGameplayTagLibrary)))
    (:then_3
      (return (LocalVariables|GetOutDynamicUIExtensions)))))

(fn GetEndCallouts
  (return (Utilities|String|SelectString (Variables|Default|GetCustomVictoryCallout) (Variables|Default|GetDeviceVictoryCallout) (Utilities|String|IsEmpty (Variables|Default|GetDeviceVictoryCallout))) (Utilities|String|SelectString (Variables|Default|GetCustomDefeatCallout) (Variables|Default|GetDeviceDefeatCallout) (Utilities|String|IsEmpty (Variables|Default|GetDeviceDefeatCallout)))))

(fn GetGameEndCallouts
  (params (bUseDeviceOveriddenCallouts :type Boolean))
  (bind (_customvictorycallout _customdefeatcallout) (CallFunction|GetEndCallouts))
  (return (Utilities|Text|ToText(String) (select bUseDeviceOveriddenCallouts (Variables|Default|GetDeviceVictoryCallout) _customvictorycallout)) (Utilities|Text|ToText(String) (select bUseDeviceOveriddenCallouts (Variables|Default|GetDeviceDefeatCallout) _customdefeatcallout)) (select bUseDeviceOveriddenCallouts (Variables|Default|GetDeviceGameEndCallout) (Variables|Default|GetGameEndCallout))))

(fn GetMaxHUDTrackers
  (return (Variables|Default|GetTrackerNumberToDisplayInHUD)))

(fn GetPickupSpawnLocation
  (return (Math|Vector|vector+vector (Transformation|GetActorLocation (Actor|GetOwner)) (* (Transformation|GetActorForwardVector (Actor|GetOwner)) 100) "0,0,50.000000")))

(fn GetScoreboardStatsArray
  (for _array_element (Variables|Default|GetTrackedScoreStats)
    (bind _output (Utilities|IsValidIndex (Variables|Default|GetTrackedStats) _array_element)
      (:true
        (Utilities|Array|AddUnique (LocalVariables|GetTrackedScoreboardStats) _output))))
  (return (LocalVariables|GetTrackedScoreboardStats)))

(fn GetTeamSetting
  (bind _teamsetting (Variables|Default|GetTeamSetting))
  (if (Knobs|Team|IsCooperativeTeam _teamsetting)
    (return 1)
    (elif (Knobs|Team|IsFreeforAllTeam _teamsetting)
      (return 0)
      (elif (Knobs|Team|IsCustomTeam _teamsetting)
        (return -1)
        (else
          (return (Knobs|Team|GetTeamValue (Variables|Default|GetTeamSetting))))))))

(fn GetTrackerNumberToDisplayInHUD
  (return (Variables|Default|GetTrackerNumberToDisplayInHUD)))

(fn GetUIExtensions
  (params (bIsCuratedHub :type Boolean))
  (LocalVariables|SetFinal%20UI%20Extensions (Variables|Default|GetCommonUIExtensions))
  (if (not bIsCuratedHub)
    (GameplayTags|AppendGameplayTagContainers (LocalVariables|GetFinal%20UI%20Extensions) (GameplayTags|MakeGameplayTagContainerfromTag "(TagName=\"Minigame.UIExtension.OverviewGameInfoTab\")")))
  (Utilities|FlowControl|Sequence
    (:then_0
      (bind _output (Utilities|IsValidIndex (Variables|Default|GetAvailableUIExtensions) (Variables|Default|GetUIExtensionSetting))
        (:true
          (bind _buiextensionisvalid (CallFunction|AdditionalUIExtensionValidityChecks _output))
          (if _buiextensionisvalid
            (GameplayTags|AppendGameplayTagContainers (LocalVariables|GetFinal%20UI%20Extensions) _output)))))
    (:then_1
      (GameplayTags|AppendGameplayTagContainers (LocalVariables|GetFinal%20UI%20Extensions) (Default|GetDynamicUITags)))
    (:then_3
      (return (LocalVariables|GetFinal%20UI%20Extensions)))))

(fn HasMinigameStarted
  (bind _minigame (CallFunction|GetCurrentMinigame))
  (Utilities|IsValid _minigame
    (:Is%20Valid
      (return (Minigame|HasStarted _minigame :!class /Script/FortniteGame.FortMinigame)))
    (:Is%20Not%20Valid
      (return false))))

(fn OverrideCustomEndCallouts
  (params (DeviceVictoryCallout :type String) (DeviceDefeatCallout :type String) (DeviceGameEndCallout :type EMinigameGameEndCallout%20Enum :default "WinLose"))
  (Networking|SwitchHasAuthority
    (:Authority
      (Variables|Default|SetDeviceVictoryCallout DeviceVictoryCallout)
      (Variables|Default|SetDeviceDefeatCallout DeviceDefeatCallout)
      (Variables|Default|SetDeviceGameEndCallout DeviceGameEndCallout)
      (CallFunction|UpdateGameSettings))))

(fn OverrideGameEndCallouts
  (params (VictoryCallout :type Text) (DefeatCallout :type Text) (GameEndCallout :type EMinigameGameEndCallout%20Enum :default "WinLose"))
  (CallFunction|OverrideCustomEndCallouts (Utilities|String|ToString(Text) VictoryCallout) (Utilities|String|ToString(Text) DefeatCallout) GameEndCallout)
  (return true))

(fn ShouldShowCumulativeScoreInHUD
  (return (Variables|Default|GetbShowCumulativeScoreOnHUD)))

(fn StatFilterSubclassToLoadedArray
  (params (InSubclassArray :type Array%20of%20Fort%20Minigame%20Stat%20Filter%20Class%20References))
  (for _array_element InSubclassArray
    (Utilities|Array|Add (LocalVariables|GetLoadedStats) (Class|FortMinigameStatsComponent|GetLoadedStatfromClass (Variables|Default|GetCachedMinigameStatsComponent) _array_element)))
  (return (LocalVariables|GetLoadedStats)))

(fn TryAddEndCondition_CollectItems
  (params (EndConditions :type Array%20of%20Fort%20Minigame%20Stat%20Query%20Structures))
  (Utilities|FlowControl|Sequence
    (:then_0
      (if (!= (Variables|Default|GetNumCollectItemsForEndCondition) 0)
        (Utilities|Array|Add EndConditions (Utilities|Struct|MakeFortMinigameStatQuery (Class|FortMinigameStatsComponent|GetLoadedStatfromClass (Variables|Default|GetCachedMinigameStatsComponent) "/Game/Athena/Playset/Minigames/StatFilters/MinigameStat_Challenge_All.MinigameStat_Challenge_All_C") "Team" true "GreaterOrEqual" (Variables|Default|GetNumCollectItemsForEndCondition) (> (Variables|Default|GetNumCollectItemsForEndCondition) 0)))
        (LocalVariables|SetbAdded true)))
    (:then_1
      (return (LocalVariables|GetbAdded)))))

(fn TryAddEndCondition_DestroyObjects
  (params (EndConditions :type Array%20of%20Fort%20Minigame%20Stat%20Query%20Structures))
  (Utilities|FlowControl|Sequence
    (:then_0
      (if (!= (Variables|Default|GetNumObjectDestroysForWin) 0)
        (Utilities|Array|Add EndConditions (Utilities|Struct|MakeFortMinigameStatQuery (Class|FortMinigameStatsComponent|GetLoadedStatfromClass (Variables|Default|GetCachedMinigameStatsComponent) "/Game/Athena/Playset/Minigames/StatFilters/MinigameStat_Destructible_All.MinigameStat_Destructible_All_C") "Team" true "GreaterOrEqual" (Variables|Default|GetNumObjectDestroysForWin) (> (Variables|Default|GetNumObjectDestroysForWin) 0)))
        (LocalVariables|SetbAdded true)))
    (:then_1
      (return (LocalVariables|GetbAdded)))))

(fn UpdateDebugSettings
  (ScalableFloat|SetValuewithContext (Variables|Default|GetDebugSettingsEnabledRow) (Variables|Default|GetbDebugSettingsEnabled))
  (ScalableFloat|SetValuewithContext (Variables|Default|GetNavMeshVisualizationEnabledRow) (Variables|Default|GetbNavMeshVisualizationEnabled))
  (ScalableFloat|SetValuewithContext (Variables|Default|GetAIPathsVisualizationEnabledRow) (Variables|Default|GetbAIPathsVisualizationEnabled))
  (ScalableFloat|SetValuewithContext (Variables|Default|GetAIGhostEnabledRow) (Variables|Default|GetbAIGhostEnabled))
  (ScalableFloat|SetValuewithContext (Variables|Default|GetInvincibilityEnabledRow) (Variables|Default|GetbInvincibilityEnabled))
  (ScalableFloat|SetValuewithContext (Variables|Default|GetVerseDebugDrawEnabledRow) (Variables|Default|GetbVerseDebugDrawEnabled))
  (ScalableFloat|SetValuewithContext (Variables|Default|GetFastIterationEnabledRow) (Variables|Default|GetbFastIterationEnabled)))

(fn UpdateAutoEndTeamThreshold
  (Networking|SwitchHasAuthority
    (:Authority
      (bind _output_get (LocalVariables|SetMinigame (CallFunction|GetCurrentMinigame)))
      (Utilities|IsValid _output_get
        (:Is%20Valid
          (Variables|/Script/FortniteGame.FortMinigame|SetAutoEndTeamThreshold (Variables|Default|GetAutoEndTeamThreshold) _output_get))))))

(fn UpdateAutoStart
  (Networking|SwitchHasAuthority
    (:Authority
      (bind _output_get (LocalVariables|SetCurrentMinigame (CallFunction|GetCurrentMinigame)))
      (Utilities|IsValid _output_get
        (:Is%20Valid
          (Utilities|IsValid (Variables|/Script/FortniteGame.FortMinigame|GetWarmupComponent _output_get)
            (:Is%20Valid
              (Variables|/Script/FortniteGame.FortMinigameWarmupComponent|SetAutoStartDelay (Variables|Default|GetAutoStartDelay) (Variables|/Script/FortniteGame.FortMinigame|GetWarmupComponent _output_get)))))))))

(fn UpdateClassSettings
  (bind _output_get (LocalVariables|SetCurrentMinigame (CallFunction|GetCurrentMinigame)))
  (Utilities|IsValid _output_get
    (:Is%20Valid
      (Utilities|IsValid (Minigame|GetFNEClassComponent _output_get)
        (:Is%20Valid
          (Variables|/Script/FortniteGame.FNEClassComponent|SetDefaultClassSlot (select (>= (Variables|Default|GetDefaultClassSlot) 0) (Math|Conversions|ToByte(Integer) (Variables|Default|GetDefaultClassSlot)) (Utilities|Enum|LiteralenumEFortMinigameClassSlot "None")) (Minigame|GetFNEClassComponent _output_get))
          (Variables|/Script/FortniteGame.FNEClassComponent|SetClassResetType (Variables|Default|GetClassResetType) (Minigame|GetFNEClassComponent _output_get)))))))

(fn UpdateEndConditions
  (Networking|SwitchHasAuthority
    (:Authority
      (bind _output_get (LocalVariables|SetMinigame (CallFunction|GetCurrentMinigame)))
      (Utilities|IsValid _output_get
        (:Is%20Valid
          (Utilities|FlowControl|Sequence
            (:then_0
              (bind _outputpin (Utilities|AddEndConditionifValueIsGTZ (Variables|Default|GetNumKillsForEndCondition) (LocalVariables|GetEndConditions) "/Game/Athena/Playset/Minigames/StatFilters/MinigameStat_PlayerKills.MinigameStat_PlayerKills_C"))
              (bind _outputpin_2 (Utilities|AddEndConditionifValueIsGTZ (Variables|Default|GetNumAIKillsForEndCondition) _outputpin "/Game/Athena/Playset/Minigames/StatFilters/MinigameStat_AIKills.MinigameStat_AIKills_C"))
              (bind _outputpin_3 (Utilities|AddEndConditionifValueIsGTZ (Variables|Default|GetScoreForEndCondition) _outputpin_2 "/Game/Athena/Playset/Minigames/StatFilters/MinigameStat_Score.MinigameStat_Score_C"))
              (CallFunction|TryAddEndConditionCollectItems _outputpin_3)
              (CallFunction|TryAddEndConditionDestroyObjects _outputpin_3))
            (:then_1
              (Variables|/Script/FortniteGame.FortMinigameEndConditionComponent|SetBaseEndConditionArray (LocalVariables|GetEndConditions) (Variables|/Script/FortniteGame.FortMinigame|GetEndConditionComponent _output_get)))))))))

(fn UpdateGameMode
  (bind _currentminigame (LocalVariables|GetCurrentMinigame))
  (bind _returnvalue (Minigame|GetFNEUIComponent _currentminigame))
  (bind _localfneuicomponent (LocalVariables|GetLocalFNEUIComponent))
  (CallFunction|FillTrackedStats)
  (bind _output_get (LocalVariables|SetCurrentMinigame (CallFunction|GetCurrentMinigame)))
  (Utilities|IsValid _output_get
    (:Is%20Valid
      (Utilities|FlowControl|Sequence
        (:then_0
          (bind _output_get_2 (Variables|Default|SetCachedMinigameStatsComponent (Minigame|Stats|GetMinigameStatsComponent (Variables|/Script/FortniteGame.FortMinigame|GetVolume _output_get))))
          (Utilities|IsValid _output_get_2
            (:Is%20Valid
              (Utilities|IsValid _returnvalue
                (:Is%20Valid
                  (LocalVariables|SetLocalFNEUIComponent _returnvalue)
                  (Minigame|Stats|SetTrackedStats _output_get_2 (Variables|Default|GetTrackedStats))
                  (bind _outarray (CallFunction|StatFilterSubclasstoLoadedArray (CallFunction|GetScoreboardStatsArray)))
                  (Variables|/Script/FortniteGame.FortMinigameStatsComponent|SetScoreboardStatObjects _outarray _output_get_2)
                  (bind _scoreboardcolumnorderstats (CallFunction|GetScoreboardOrderObjects (Variables|Default|GetTrackedScoreColumnOrder) (Variables|/Script/FortniteGame.FortMinigameStatsComponent|GetTrackedStatObjects _output_get_2)))
                  (Minigame|Stats|SetScoreboardColumnOrder _output_get_2 _scoreboardcolumnorderstats)
                  (Minigame|Stats|SetTimeAliveTeamTrackingMethod _output_get_2 (Variables|Default|GetTimeAliveTeamTrackingMethod))
                  (Minigame|Stats|SetTimeAliveStartPoint _output_get_2 (Variables|Default|GetTimeAliveStartPoint))
                  (Class|FNEUIComponent|SetUIExtensionTags _localfneuicomponent (CallFunction|GetUIExtensions (Minigame|IsCuratedHub _output_get)) :!class /Script/FortniteGame.FNEUIComponent)
                  (Class|FNEUIComponent|SetMinigameMapWidget _localfneuicomponent (Variables|Default|GetMinigameFullscreenMapWidgetType) :!class /Script/FortniteGame.FNEUIComponent)
                  (Class|FNEUIComponent|SetShowMapWidget _localfneuicomponent (Variables|Default|GetbShouldDisplayMap) :!class /Script/FortniteGame.FNEUIComponent)
                  (Variables|/Script/FortniteGame.FortMinigame|SetWinCondition (Variables|Default|GetWinCondition) _output_get)
                  (Variables|/Script/FortniteGame.FortMinigameEndConditionComponent|SetbAllTeamsMustMatchEndConditions (Math|Conversions|ToBoolean(Integer) (Variables|Default|GetAllTeamsMustMatchEndConditions)) (Variables|/Script/FortniteGame.FortMinigame|GetEndConditionComponent _output_get))
                  (Variables|/Script/FortniteGame.FortMinigame|SetbEndGameOnMatchPointWin (Math|Conversions|ToBoolean(Integer) (Variables|Default|GetEndGameOnMatchPointWin)) _output_get)
                  (Variables|/Script/FortniteGame.FortMinigame|SetPlayerPersistence (Variables|Default|GetPlayerPersistence) _output_get)
                  (Variables|/Script/FortniteGame.FortMinigame|SetbOnlyAllowRespawningIfPlayerStartPadsFound (Math|Conversions|ToBoolean(Integer) (Variables|Default|GetOnlyAllowRespawningIfPlayerStartPadsFound)) _output_get)
                  (UI|SetPipHudAdEnabled (LocalVariables|GetLocalFNEUIComponent) (not (Variables|Default|GetbDisablePiPHUDAdvert))))))))
        (:then_1
          (bind _scoreboardcomponent (Variables|/Script/FortniteGame.FortMinigame|GetScoreboardComponent _output_get)
            (:then
              (Variables|/Script/FortniteGame.FortMinigameScoreboardComponent|SetbSortScoreboardEntries true _scoreboardcomponent)
              (Variables|/Script/FortniteGame.FortMinigameScoreboardComponent|SetbShowCumulativeScoreboard (Math|Conversions|ToBoolean(Integer) (Variables|Default|GetShowCumulativeScoreboard)) _scoreboardcomponent)
              (Variables|/Script/FortniteGame.FortMinigameScoreboardComponent|SetbForceIndividualScoresInTeamGame (Variables|Default|GetbForceIndividualScores) _scoreboardcomponent)
              (Variables|/Script/FortniteGame.FortMinigameScoreboardComponent|SetbIncludePlayersWhoLeftGameInScoreboard (Variables|Default|GetbIncludePlayersWhoLeftGameInScoreboard) _scoreboardcomponent))
            (:else)))))))

(fn UpdateGameSettings
  (bind _currentminigame (LocalVariables|GetCurrentMinigame))
  (bind _victoryscreencomponent (Variables|/Script/FortniteGame.FortMinigame|GetVictoryScreenComponent _currentminigame))
  (bind _output_get (LocalVariables|SetCurrentMinigame (CallFunction|GetCurrentMinigame)))
  (Utilities|IsValid _output_get
    (:Is%20Valid
      (Utilities|FlowControl|Sequence
        (:then_0
          (bind (_victorycallout _defaultcallout _gameendcallout) (UI|GetGameEndCallouts true :!class /Game/Creative/Devices/CreativeMinigameUIInterface.CreativeMinigameUIInterface_C))
          (Variables|/Script/FortniteGame.FortMinigameVictoryScreenComponent|SetGameEndCallout _gameendcallout _victoryscreencomponent)
          (Variables|/Script/FortniteGame.FortMinigameVictoryScreenComponent|SetVictoryAudioIndex (Variables|Default|GetVictoryAudioIndex) _victoryscreencomponent)
          (Variables|/Script/FortniteGame.FortMinigameVictoryScreenComponent|SetDrawAudioIndex (Variables|Default|GetDrawAudioIndex) _victoryscreencomponent))
        (:then_1
          (Variables|/Script/FortniteGame.FortMinigameVictoryScreenComponent|SetDefeatAudioIndex (Variables|Default|GetDefeatAudioIndex) _victoryscreencomponent)
          (Variables|/Script/FortniteGame.FortMinigameVictoryScreenComponent|SetPostGameSettings (Utilities|Struct|MakeMinigamePostGameSettings (Variables|Default|GetPostGameType) (Math|Conversions|ToFloat(Integer) (Variables|Default|GetPostGameDelay)) (Variables|Default|GetbPostGameCustomShowScoreboard) (Utilities|Struct|MakeMinigamePostGameAnimationSettings (Variables|Default|GetPostGameCustomVictoryAnimationStyle) (Variables|Default|GetPostGameCustomVictoryAnimationColorSet) (Variables|Default|GetPostGameCustomVictoryText) (Variables|Default|GetPostGameCustomVictorySubText)) (Utilities|Struct|MakeMinigamePostGameAnimationSettings (Variables|Default|GetPostGameCustomDefeatAnimationStyle) (Variables|Default|GetPostGameCustomDefeatAnimationColorSet) (Variables|Default|GetPostGameCustomDefeatText) (Variables|Default|GetPostGameCustomDefeatSubText)) (Utilities|Struct|MakeMinigamePostGameAnimationSettings (Variables|Default|GetPostGameCustomTieAnimationStyle) (Variables|Default|GetPostGameCustomTieAnimationColorSet) (Variables|Default|GetPostGameCustomTieText) (Variables|Default|GetPostGameCustomTieSubText))) _victoryscreencomponent))))))

(fn UpdateGameStartCountdown
  (bind _minigame (CallFunction|GetCurrentMinigame))
  (bind _warmupcomponent (Variables|/Script/FortniteGame.FortMinigame|GetWarmupComponent _minigame))
  (Networking|SwitchHasAuthority
    (:Authority
      (Utilities|IsValid _minigame
        (:Is%20Valid
          (Utilities|IsValid _warmupcomponent
            (:Is%20Valid
              (Variables|/Script/FortniteGame.FortMinigameWarmupComponent|SetWarmupDuration (Variables|Default|GetStartCountdownDuration) _warmupcomponent))))))))

(fn UpdateJoinInProgress
  (Networking|SwitchHasAuthority
    (:Authority
      (bind _output_get (LocalVariables|SetNewLocalVar_1 (CallFunction|GetCurrentMinigame)))
      (Utilities|IsValid _output_get
        (:Is%20Valid
          (Variables|/Script/FortniteGame.FortMinigame|SetbAllowJoinInProgress (Utilities|Enum|NotEqual(Enum) (Variables|UserOptions|GetJoinInProgressBehavior) "WatchOnly") _output_get))))))

(fn UpdateMinigameMutators
  (Networking|SwitchHasAuthority
    (:Authority
      (bind _output_get (LocalVariables|SetCurrentMinigame (CallFunction|GetCurrentMinigame)))
      (Utilities|IsValid _output_get
        (:Is%20Valid
          (bind _output_get_2 (LocalVariables|SetMutatorList (Actor|GetComponentbyClass _output_get "/Script/FortniteGame.FortMutatorListComponent")))
          (Modifiers|SetPropertyOverrides _output_get_2 (OptionData|GetPropertyOverrides (Variables|Default|GetToyOptionsComponent))))))))

(fn UpdatePhysicsSettings
  (bind _output_get (LocalVariables|SetbEnablePhysicsImprovements (ScalableFloat|GetValueasBoolwithContext (Variables|Default|GetPhysicsActivationOnImpactSetting))))
  (ScalableFloat|SetValuewithContext (Variables|Default|GetPhysicsUserDataSetupSetting) _output_get)
  (ScalableFloat|SetValuewithContext (Variables|Default|GetPropToPropPhysicsSetting) _output_get)
  (ScalableFloat|SetValuewithContext (Variables|Default|GetVehicleToPropPhysicsSetting) _output_get))

(fn UpdatePlayerLives
  (Networking|SwitchHasAuthority
    (:Authority
      (bind _output_get (LocalVariables|SetNewLocalVar_0 (CallFunction|GetCurrentMinigame)))
      (Utilities|IsValid _output_get
        (:Is%20Valid
          (Variables|/Script/FortniteGame.FortMinigame|SetMaxLivesForPlayer (Variables|Default|GetPlayerLives) _output_get))))))

(fn UpdatePlayerPersistence
  (bind _minigame (CallFunction|GetCurrentMinigame))
  (Networking|SwitchHasAuthority
    (:Authority
      (Utilities|IsValid _minigame
        (:Is%20Valid
          (Variables|/Script/FortniteGame.FortMinigame|SetPlayerPersistence (Variables|Default|GetPlayerPersistence) _minigame))))))

(fn UpdatePostGameSpawnLocationSetting
  (bind _minigame (CallFunction|GetCurrentMinigame))
  (Networking|SwitchHasAuthority
    (:Authority
      (LocalVariables|SetCurrentMinigame _minigame)
      (Utilities|IsValid _minigame
        (:Is%20Valid
          (Variables|/Script/FortniteGame.FortMinigame|SetPostGameSpawnLocationSetting (Variables|Default|GetPostGameSpawnLocationSetting) _minigame))))))

(fn UpdateSpawnLocationSetting
  (bind _minigame (CallFunction|GetCurrentMinigame))
  (Networking|SwitchHasAuthority
    (:Authority
      (LocalVariables|SetCurrentMinigame _minigame)
      (Utilities|IsValid _minigame
        (:Is%20Valid
          (Variables|/Script/FortniteGame.FortMinigame|SetSpawnLocationSetting (Variables|Default|GetSpawnLocationSetting) _minigame))))))

(fn UpdateSpawnPadSelection
  (bind _minigame (CallFunction|GetCurrentMinigame))
  (Networking|SwitchHasAuthority
    (:Authority
      (LocalVariables|SetMinigame _minigame)
      (Utilities|IsValid _minigame
        (:Is%20Valid
          (Variables|/Script/FortniteGame.FortMinigame|SetSpawnPadSelection (Variables|Default|GetSpawnPadSelection) _minigame))))))

(fn UpdateTeamSetting
  (bind _returnvalue (Minigame|GetMinigameTeamComponent (CallFunction|GetCurrentMinigame)))
  (bind _minigame (CallFunction|GetCurrentMinigame))
  (bind _returnvalue_2 (Minigame|GetMinigameTeamBalanceComponent _minigame))
  (Networking|SwitchHasAuthority
    (:Authority
      (Utilities|IsValid _minigame
        (:Is%20Valid
          (Utilities|FlowControl|Sequence
            (:then_0
              (Utilities|IsValid _returnvalue
                (:Is%20Valid
                  (Minigame|Teams|SetMinigameNumTeams _returnvalue (Variables|Default|GetTeamSetting)))))
            (:then_1
              (Utilities|IsValid _returnvalue_2
                (:Is%20Valid
                  (Variables|/Script/FortniteGame.FortMinigameTeamBalanceComponent|SetTeamRotationSetting (Variables|Default|GetTeamRotationSetting) _returnvalue_2)
                  (Minigame|Teams|SetStableTeamCosmetics _returnvalue_2 (Variables|Default|GetbStableTeamCosmetics)))))))))))

(fn UpdateTeamToMoveToWhenOutOfSpawns
  (bind _teamtomovetowhenoutofspawns (Variables|Default|GetTeamToMoveToWhenOutOfSpawns))
  (Networking|SwitchHasAuthority
    (:Authority
      (bind _output_get (LocalVariables|SetMinigame (CallFunction|GetCurrentMinigame)))
      (Utilities|IsValid _output_get
        (:Is%20Valid
          (if (Knobs|Team|IsSpectatingTeam _teamtomovetowhenoutofspawns)
            (Variables|/Script/FortniteGame.FortMinigame|SetTeamToSwapToWhenOutOfSpawns 0 _output_get)
            (else
              (Variables|/Script/FortniteGame.FortMinigame|SetTeamToSwapToWhenOutOfSpawns (Math|Byte|Byte+Byte (Utilities|Enum|LiteralenumEFortTeam "Monster") (Math|Conversions|ToByte(Integer) (Knobs|Team|GetTeamValue _teamtomovetowhenoutofspawns))) _output_get))))))))

(fn UpdateTimeLimit
  (Networking|SwitchHasAuthority
    (:Authority
      (bind _output_get (LocalVariables|SetCurrentMinigame (CallFunction|GetCurrentMinigame)))
      (Utilities|IsValid _output_get
        (:Is%20Valid
          (Utilities|FlowControl|Sequence
            (:then_0
              (Variables|/Script/FortniteGame.FortMinigame|SetTimeLimit (Variables|Default|GetTimeLimit) _output_get)
              (Variables|/Script/FortniteGame.FortMinigame|SetbTimerCountsDown (== (Variables|Default|GetTimerDirection) 0) _output_get)
              (Variables|/Script/FortniteGame.FortMinigame|SetbStopwatchMode (Variables|Default|GetbStopwatchMode) _output_get))
            (:then_1
              (bind _scoreboardcomponent (Variables|/Script/FortniteGame.FortMinigame|GetScoreboardComponent _output_get)
                (:then
                  (Variables|/Script/FortniteGame.FortMinigameScoreboardComponent|SetGameWinnerDisplayTime (Variables|Default|GetWinnerDisplayTime) _scoreboardcomponent)
                  (Variables|/Script/FortniteGame.FortMinigameScoreboardComponent|SetGameScoreDisplayTime (Variables|Default|GetScoreDisplayTime) _scoreboardcomponent)
                  (Variables|/Script/FortniteGame.FortMinigameScoreboardComponent|SetRoundWinnerDisplayTime (Variables|Default|GetRoundWinnerDisplayTime) _scoreboardcomponent)
                  (Variables|/Script/FortniteGame.FortMinigameScoreboardComponent|SetRoundScoreDisplayTime (Variables|Default|GetRoundScoreDisplayTime) _scoreboardcomponent))
                (:else)))))))))

(fn UpdateTotalRounds
  (bind _minigame (CallFunction|GetCurrentMinigame))
  (bind _returnvalue (Minigame|GetFNEUIComponent _minigame))
  (Networking|SwitchHasAuthority
    (:Authority
      (Utilities|IsValid _minigame
        (:Is%20Valid
          (Utilities|IsValid _returnvalue
            (:Is%20Valid
              (bind _output_get (Variables|/Script/FortniteGame.FortMinigame|SetTotalRounds (Variables|Default|GetTotalRounds) _minigame))
              (if (> _output_get 1)
                (GameplayTags|AddGameplayTag (Variables|/Script/FortniteGame.FNEUIComponent|GetUIExtensionTags _returnvalue) "(TagName=\"Minigame.UIExtension.Rounds\")" :!class /Script/GameplayTags.BlueprintGameplayTagLibrary)))))))))

(fn ConstructionScript)
